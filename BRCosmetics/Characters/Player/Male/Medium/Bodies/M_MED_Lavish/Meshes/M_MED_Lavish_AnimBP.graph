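# Animation Blueprint for the M_MED_Lavish player body, posing its front straps and side gun over the pose copied from the attached parent mesh.
# Weighted constraints drive the straps, then Anim Dynamics swings the gun.
# Left upper-arm pitch scales corrective gun and left-strap offsets, zeroed when crouching or accelerating forward.
# When the frontend idle curve is positive, the output switches to a recorded holster pose layered on those bones.
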AnimBlueprint /BRCosmetics/Characters/Player/Male/Medium/Bodies/M_MED_Lavish/Meshes/M_MED_Lavish_AnimBP
target_skeleton=&/M_MED_Lavish_Skeleton
graphs [
 AnimGraph {nodes [
   n1 = "Constraint - dyn_L_frontStrap_follow_B (PPPPPPPP)"!(
    ComponentPose=Local To Component(LocalPose=Copy Pose From Mesh() {bUseAttachedParent=true})
    Alpha=1
    ConstraintWeights_0=0.1966
    ConstraintWeights_1:float=0.0967
    ConstraintWeights_2:float=0.0028
    ConstraintWeights_3:float=0.0031
    ConstraintWeights_4:float=0.1148
    ConstraintWeights_5:float=0.2572
    ConstraintWeights_6:float=0.3288
    ConstraintWeights_7:float=0
   ) -> (Pose) {
    alphaBoolBlend {blendOption=Linear}
    boneToModify {boneName=dyn_L_frontStrap_follow_B}
    constraintSetup [
     {targetBone {boneName=clavicle_l}}
     {targetBone {boneName=upperarm_twist_02_l}}
     {targetBone {boneName=spine_01}}
     {targetBone {boneName=spine_03}}
     {targetBone {boneName=spine_04}}
     {targetBone {boneName=spine_05}}
     {targetBone {boneName=upperarm_twist_01_l}}
     {targetBone {boneName=pelvis}}
    ]
    constraintWeights [1,1,1,1,1,1,1,1]
   }
   n2 = "Constraint - dyn_R_frontStrap_follow_B (PPPPPPPPP)"!(
    ComponentPose=@n1
    Alpha=1
    ConstraintWeights_0=0.3334
    ConstraintWeights_1:float=0.0339
    ConstraintWeights_2:float=0.0086
    ConstraintWeights_3:float=0.0059
    ConstraintWeights_4:float=0.1315
    ConstraintWeights_5:float=0.2412
    ConstraintWeights_6:float=0.2435
    ConstraintWeights_7:float=0.0016
    ConstraintWeights_8:float=0.0003
   ) -> (Pose) {
    alphaBoolBlend {blendOption=Linear}
    boneToModify {boneName=dyn_R_frontStrap_follow_B}
    constraintSetup [
     {targetBone {boneName=upperarm_twist_01_r}}
     {targetBone {boneName=upperarm_twist_02_r}}
     {targetBone {boneName=spine_01}}
     {targetBone {boneName=spine_03}}
     {targetBone {boneName=spine_04}}
     {targetBone {boneName=spine_05}}
     {targetBone {boneName=clavicle_r}}
     {targetBone {boneName=pelvis}}
     {targetBone {boneName=neck_01}}
    ]
    constraintWeights [1,1,1,1,1,1,1,1,1]
   }
   n3 = "Constraint - dyn_L_frontStrap_follow_A (PPPPPPPP)"!(
    ComponentPose=@n2
    Alpha=1
    ConstraintWeights_0=0.231
    ConstraintWeights_1:float=0.0362
    ConstraintWeights_2:float=0.0012
    ConstraintWeights_3:float=0.0015
    ConstraintWeights_4:float=0.0692
    ConstraintWeights_5:float=0.385
    ConstraintWeights_6:float=0.276
    ConstraintWeights_7:float=0
   ) -> (Pose) {
    alphaBoolBlend {blendOption=Linear}
    boneToModify {boneName=dyn_L_frontStrap_follow_A}
    constraintSetup [
     {targetBone {boneName=clavicle_l}}
     {targetBone {boneName=upperarm_twist_02_l}}
     {targetBone {boneName=spine_01}}
     {targetBone {boneName=spine_03}}
     {targetBone {boneName=spine_04}}
     {targetBone {boneName=spine_05}}
     {targetBone {boneName=upperarm_twist_01_l}}
     {targetBone {boneName=pelvis}}
    ]
    constraintWeights [1,1,1,1,1,1,1,1]
   }
   n4 = "Constraint - dyn_R_frontStrap_follow_A (PPPPPPPPPP)"!(
    ComponentPose=@n3
    Alpha=1
    ConstraintWeights_0=0.2489
    ConstraintWeights_1:float=0.0143
    ConstraintWeights_2:float=0.0028
    ConstraintWeights_3:float=0.0013
    ConstraintWeights_4:float=0.0762
    ConstraintWeights_5:float=0.2928
    ConstraintWeights_6:float=0.0082
    ConstraintWeights_7:float=0.334
    ConstraintWeights_8:float=0.0001
    ConstraintWeights_9:float=0.0214
   ) -> (Pose) {
    alphaBoolBlend {blendOption=Linear}
    boneToModify {boneName=dyn_R_frontStrap_follow_A}
    constraintSetup [
     {targetBone {boneName=upperarm_twist_01_r}}
     {targetBone {boneName=upperarm_twist_02_r}}
     {targetBone {boneName=spine_01}}
     {targetBone {boneName=spine_03}}
     {targetBone {boneName=spine_04}}
     {targetBone {boneName=spine_05}}
     {targetBone {boneName=neck_02}}
     {targetBone {boneName=clavicle_r}}
     {targetBone {boneName=pelvis}}
     {targetBone {boneName=neck_01}}
    ]
    constraintWeights [1,1,1,1,1,1,1,1,1,1]
   }
   n5 = Anim Dynamics - Bone: dyn_sideGun(ComponentPose=@n4, Alpha=1, ExternalForce="0,0,0") {
    angularDampingOverride=0.9999
    angularSpringConstant=50
    bAngularSpring=true
    bGravityOverrideInSimSpace=true
    bOverrideAngularDamping=true
    gravityScale=0.5
    alphaBoolBlend {blendOption=Linear}
    alphaScaleBiasClamp {bias=0.35, scale=0.08}
    boundBone {boneName=dyn_sideGun}
    componentLinearAccScale=<1 0.3 0.8>
    componentLinearVelScale=<0.1 0.1 0.01>
    physicsBodyDefinitions [{
      sphereCollisionRadius=0
      boundBone {boneName=dyn_sideGun}
      boxExtents=<5 10 20>
      constraintSetup {angularLimitsMax=<50 20 30>, angularLimitsMin=<-50 -60 -30>, angularTarget=<0 0 0>}
      localJointOffset=<0 0 -2>
     }]
   }
   n6 = Get L_upperarm_alpha() -> (L_upperarm_alpha:double)
   n7 = "Transform (Modify) Bone - Bone: dyn_sideGun_master"(ComponentPose=@n5, Alpha=@n6, Translation="-2.5,0,0", Rotation="0.000,0.000,20", Scale="1,1,1") {
    rotationMode=BMM_Replace
    rotationSpace=BCS_BoneSpace
    translationMode=BMM_Additive
    translationSpace=BCS_BoneSpace
    alphaBoolBlend {blendOption=Linear}
    alphaScaleBiasClamp {bInterpResult=true}
    boneToModify {boneName=dyn_sideGun_master}
    rotation {roll=20}
    translation {x=-2.5}
   }
   n8 = "Transform (Modify) Bone - Bone: dyn_L_frontStrap_follow_B"(ComponentPose=@n7, Alpha=@n6, Translation="2,0,0", Rotation="0,0,0", Scale="1,1,1") {
    translationMode=BMM_Additive
    translationSpace=BCS_BoneSpace
    alphaBoolBlend {blendOption=Linear}
    alphaScaleBiasClamp {bInterpResult=true}
    boneToModify {boneName=dyn_L_frontStrap_follow_B}
    translation {x=2}
   }
   n9 = Layered blend per bone(
    BasePose=Copy Pose From Mesh() {bCopyCurves=true, bUseAttachedParent=true}
    BlendPoses_0=Sequence Player 'GunHolsterPose_record'() {sequence=/BRCosmetics/Animation/Game/MainPlayer/Menu/BR/CustomSkins/_S17/Lavish/GunHolsterPose_record}
   ) {layerSetup [{branchFilters [
       {boneName=dyn_sideGun_master}
       {boneName=dyn_L_frontStrap_follow_A}
       {boneName=dyn_L_frontStrap_follow_B}
       {boneName=dyn_R_frontStrap_follow_B}
      ]}]}
   Output Pose(Result=Blend Poses by bool(BlendPose_0=@n9, BlendPose_1=Component To Local(ComponentPose=@n8), bActiveValue=Get PlayingFrontendIdle() -> (PlayingFrontendIdle:bool)) {blendType=Linear}) {layerGroup=None}
  ]}
 EventGraph {nodes [
   Event BlueprintInitializeAnimation()
   > Set SkelMesh(SkelMesh:Skeletal Mesh Object Reference=&/M_MED_Lavish) -> (Output_Get:Skeletal Mesh Object Reference)
   > n1 = Cast To FortPlayerPawn(Object=TryGetFortPawn())
   > Set Pawn(Pawn:Fort Player Pawn Object Reference=@n1) -> (Output_Get:Fort Player Pawn Object Reference)
   Event BlueprintUpdateAnimation()
   > n2 = Sequence!(execute) -> (then_0, then_1, then_2:Exec, then_3:Exec)
   n3 = GetAnimInstance(self=GetOwnerSkeletalMeshComponent())
   > Set PlayingFrontendIdle(PlayingFrontendIdle:bool="float > float"(A=GetCurveValue(self=@n3, CurveName=FrontEnd_Idle_Bypass), B=0) {is=title K2Node_CallFunction}) -> (Output_Get:bool)
   n4 = BreakTransform(InTransform=GetDeltaTransformFromRefPose(self=GetOwnerSkeletalMeshComponent(), BoneName=upperarm_l, BaseName=clavicle_l))
   Set L_upperarm_alpha(execute=@n2.then_1, L_upperarm_alpha:double=MapRangeClamped(Value=BreakRotator(InRot=@n4.Rotation).Pitch, InRangeA=30, InRangeB=-5)) -> (Output_Get:double)
   n5 = Branch(
    execute=@n2.then_2
    Condition=NOR Boolean!(A:bool=Get bIsCrouching(), B:bool=Get isAcceleratingForward() -> (isAcceleratingForward:bool)) -> (ReturnValue:bool)
   )
   > Set L_upperarm_alpha(L_upperarm_alpha=Get L_upperarm_alpha())
   Set L_upperarm_alpha(execute=@n5.else, L_upperarm_alpha=0)
   Get!(execute:Exec=@n2.then_3) -> (then:Exec, else:Exec, SkelMesh:Skeletal Mesh Object Reference)
   > Get!(execute) -> (then, else, Pawn:Fort Player Pawn Object Reference)
   > n6 = PlayerMovement(AnimInstance=Self-Reference(), XDotProductCoefficient=25, YDotProductCoefficient=25, ZDotProductCoefficient=1)
   > Set isAcceleratingForward(isAcceleratingForward:bool="float > float"(A=@n6.YMovement, B=0)) -> (Output_Get:bool)
   PrintString!(execute, InString="Its TRUE!", bPrintToScreen, bPrintToLog, TextColor, Duration, Key) -> (then)
  ]}
]
variables [
 isAcceleratingForward {type=boolean}
 L_upperarm_alpha {type=number}
 Pawn {type=FortPlayerPawn}
 PlayingFrontendIdle {type=boolean}
 SkelMesh {type=SkeletalMesh, default=null}
]
